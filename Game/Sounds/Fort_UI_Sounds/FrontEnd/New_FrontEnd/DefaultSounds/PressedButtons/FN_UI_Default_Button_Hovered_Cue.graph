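# Sound cue for the default button hover sound in the front-end UI, routed to the Fort_UI sound class.
# A single wave player feeds UI_MP_Hover_Default_01 into a modulator whose pitch and volume ranges are fixed at unity, so playback has no random variation.
# A shared hover-tick concurrency rule, named for one voice with stop-oldest, limits overlapping hover sounds.
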SoundCue /Game/Sounds/Fort_UI_Sounds/FrontEnd/New_FrontEnd/DefaultSounds/PressedButtons/FN_UI_Default_Button_Hovered_Cue
root=Modulator_0
nodes [
 {id=Modulator_0, children [WavePlayer_0], properties {pitchMax=1, pitchMin=1, volumeMax=1, volumeMin=1}}
 {id=WavePlayer_0, properties {soundWaveAssetPtr=/Game/Sounds/Fort_UI_Sounds/FrontEnd/DefaultSounds/PressedButtons/wav/UI_MP_Hover_Default_01}}
]
properties {
 maxDistance=2097152
 soundClassObject=&Fort_UI
 volumeMultiplier=1
 concurrencyOverrides {maxCount=1, resolutionRule=StopOldest}
 concurrencySet [/Game/Sounds/Concurrency_Rules/FrontEnd_UI_Menu_Max_01_Stop_Oldest_Hover_Tick]
}
