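# Sound cue for the default button press in the Fortnite front-end UI.
# A single wave player feeds the output directly with the main button click wave, with no randomization or modulation.
# The cue plays in the Fort_UI sound class with boosted volume and references a shared front-end menu concurrency rule named for stopping the oldest sounds.
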SoundCue /Game/Sounds/Fort_UI_Sounds/FrontEnd/New_FrontEnd/DefaultSounds/PressedButtons/FN_UI_Default_Button_Pressed_Cue
root=WavePlayer_1
nodes [{id=WavePlayer_1, properties {soundWaveAssetPtr=/Game/Sounds/Fort_UI_Sounds/FrontEnd/New_FrontEnd/UI_FN_Button_Click_Main_01}}]
properties {
 maxDistance=2097152
 soundClassObject=&Fort_UI
 volumeMultiplier=1.5
 concurrencyOverrides {maxCount=1, resolutionRule=StopOldest}
 concurrencySet [/Game/Sounds/Concurrency_Rules/FrontEnd_UI_Menu_Max_03_Stop_Oldest]
}
